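# MetaSound patch that spatializes a mono source to binaural left/right from azimuth and distance.
# An ITD Panner takes its angle from MSP_AzimuthMapper and its distance factor from distance mapped against FalloffDistance.
# Per-channel crossfades blend attenuated, unpanned mono into the panned signal as distance moves from BinauralRadiusStart to BinauralRadiusEnd.
# A small offset on the end radius keeps equal radii from forming a zero-width range.
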MetaSoundPatch /FortMetasound/Attenuation/MSP_Attenuation_Spatialization_Binaural
inputs [
 Azimuth {type=Float}
 BinauralRadiusEnd {type=Float}
 BinauralRadiusStart {type=Float}
 Distance {type=Float}
 FalloffDistance {type=Float}
 Mono.In {type=Audio}
]
outputs [
 Left {type=Audio}
 Right {type=Audio}
]
pages [Default {
  input_defaults {Azimuth=0, BinauralRadiusEnd=0, BinauralRadiusStart=0, Distance=0, FalloffDistance=1000, Mono.In=null}
  nodes [
   n1 = InitVariable.Float()
   n2 = MapRange.MapRange:Float(In=^Distance, In Range B=^FalloffDistance)
   ITDPanner_0 = UE.ITD Panner(
    In=^Mono.In
    Angle=/FortMetasound/Patches/MSP_AzimuthMapper(Azimuth=^Azimuth) {also {defaults {Back=270, Front=90, Left=180, Right=0}}}.Out
    Distance Factor=VariableAccessor.Float(Variable=VariableMutator.Float(Variable=@n1.Variable, Value=@n2.Out Value).Variable).Value
   ) {also {defaults {Head Width=5}}}
   n3 = UE.Multiply:Audio by Float(PrimaryOperand=^Mono.In) {also {defaults {AdditionalOperands=0.55}}}
   n4 = UE.Add:Float(PrimaryOperand=^BinauralRadiusEnd) {also {defaults {AdditionalOperands=0.1}}}
   n5 = VariableAccessor.Float(Variable=VariableMutator.Float(
     Variable=InitVariable.Float().Variable
     Value=MapRange.MapRange:Float(In Range A=^BinauralRadiusStart, In=^Distance, In Range B=@n4.Out).Out Value
    ).Variable)
   CrossfadeTriggerRouteAudio2_0 = "Crossfade.Trigger Route (Audio, 2):Audio"(In 1=@ITDPanner_0.Out Left, In 0=@n3.Out, Crossfade Value=@n5.Value)
   CrossfadeTriggerRouteAudio2_1 = "Crossfade.Trigger Route (Audio, 2):Audio"(In 1=@ITDPanner_0.Out Right, In 0=@n3.Out, Crossfade Value=@n5.Value)
   InitVariable.Float()
  ]
  links [{dst="@Left", src=CrossfadeTriggerRouteAudio2_0.Out}, {dst="@Right", src=CrossfadeTriggerRouteAudio2_1.Out}]
 }]
