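# MetaSound patch that simulates distance-based air absorption by filtering incoming audio.
# Distance, clamped to the min and max range, maps logarithmically to separate low-pass and high-pass cutoffs within configurable limits.
# Threshold triggers on the minimum distance switch each cutoff between that mapped value and a fully open value.
# The audio runs through a one-pole low-pass filter, then a one-pole high-pass filter.
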MetaSoundPatch /FortMetasound/Attenuation/MSP_Attenuation_AirAbsorption
inputs [
 AudioIn {type=Audio}
 Distance {type=Float}
 HighPassCutoffFrequencyMax {type=Float}
 HighPassCutoffFrequencyMin {type=Float}
 LowPassCutoffFrequencyMax {type=Float}
 LowPassCutoffFrequencyMin {type=Float}
 MaxDistanceRange {type=Float}
 MinDistanceRange {type=Float}
]
outputs [AudioOut {type=Audio}]
pages [Default {
  input_defaults {
   AudioIn=null
   Distance=0
   HighPassCutoffFrequencyMax=0
   HighPassCutoffFrequencyMin=0
   LowPassCutoffFrequencyMax=20000
   LowPassCutoffFrequencyMin=20000
   MaxDistanceRange=6000
   MinDistanceRange=3000
  }
  nodes [
   LinearToLogFrequency_1 = UE.Linear To Log Frequency(
    Max Range=^LowPassCutoffFrequencyMax
    Min Range=^LowPassCutoffFrequencyMin
    Max Domain=^MaxDistanceRange
    Min Domain=^MinDistanceRange
    Value=Clamp.Clamp:Float(In=^Distance, Max=^MaxDistanceRange, Min=^MinDistanceRange).Value
   )
   TriggerRouteTriggerRouteFloat2_0 = "TriggerRoute.Trigger Route (Float, 2):Float"(
    Value 1=@LinearToLogFrequency_1.Frequency
    Set 1=UE.TriggerOnThreshold:Float(In=^Distance, Threshold=^MinDistanceRange).Out
    Set 0=UE.TriggerOnThreshold:Float(In=^Distance, Threshold=^MinDistanceRange) {also {defaults {Type=1}}}.Out
   ) {also {defaults {Value 0=20000}}}
   OnePoleLowPassFilter_0 = UE.One-Pole Low Pass Filter:Audio(In=^AudioIn, Cutoff Frequency=@TriggerRouteTriggerRouteFloat2_0.Value)
   LinearToLogFrequency_0 = UE.Linear To Log Frequency(
    Max Range=^HighPassCutoffFrequencyMax
    Min Range=^HighPassCutoffFrequencyMin
    Max Domain=^MaxDistanceRange
    Min Domain=^MinDistanceRange
    Value=Clamp.Clamp:Float(In=^Distance, Max=^MaxDistanceRange, Min=^MinDistanceRange).Value
   )
   TriggerRouteTriggerRouteFloat2_1 = "TriggerRoute.Trigger Route (Float, 2):Float"(
    Value 1=@LinearToLogFrequency_0.Frequency
    Set 1=UE.TriggerOnThreshold:Float(In=^Distance, Threshold=^MinDistanceRange).Out
    Set 0=UE.TriggerOnThreshold:Float(In=^Distance, Threshold=^MinDistanceRange) {also {defaults {Type=1}}}.Out
   ) {also {defaults {Value 0=0}}}
   OnePoleHighPassFilter_0 = UE.One-Pole High Pass Filter:Audio(In=@OnePoleLowPassFilter_0.Out, Cutoff Frequency=@TriggerRouteTriggerRouteFloat2_1.Value)
  ]
  links [{dst="@AudioOut", src=OnePoleHighPassFilter_0.Out}]
 }]
