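# MetaSound patch that computes a focus multiplier for attenuation from a sound's azimuth and elevation.
# It maps each absolute angle between focus and non-focus thresholds onto their values, then multiplies the results.
# A deferred variable holds the previous product to detect rises or falls, selecting attack or release interpolation time for the smoothed output.
# The enable bool, converted to float, scales both times, so disabling it zeroes them.
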MetaSoundPatch /FortMetasound/Attenuation/MSP_Attenuation_Focus
inputs [
 Azimuth {type=Float}
 bEnableFocusInterpolation {type=Bool}
 Elevation {type=Float}
 FocusAngleThreshold {type=Float}
 FocusAttackInterpSpeed {type=Float}
 FocusReleaseInterpSpeed {type=Float}
 FocusValue {type=Float}
 NonFocusAngleThreshold {type=Float}
 NonFocusValue {type=Float}
]
outputs [FocusMultiplier {type=Float}]
pages [Default {
  input_defaults {
   Azimuth=0
   Elevation=0
   FocusAngleThreshold=25
   FocusAttackInterpSpeed=1
   FocusReleaseInterpSpeed=1
   FocusValue=1
   NonFocusAngleThreshold=35
   NonFocusValue=0.3
   bEnableFocusInterpolation=null
  }
  nodes [
   InitVariable.Float()
   n1 = Convert.Float:Bool(Bool=^bEnableFocusInterpolation)
   n2 = UE.Multiply:Float(
    PrimaryOperand=MapRange.MapRange:Float(
     In Range A=^FocusAngleThreshold
     Out Range A=^FocusValue
     In Range B=^NonFocusAngleThreshold
     Out Range B=^NonFocusValue
     In=AbsoluteValue.Abs:Float(Input=^Azimuth).Value
    ).Out Value
    AdditionalOperands=MapRange.MapRange:Float(
     In Range A=^FocusAngleThreshold
     Out Range A=^FocusValue
     In Range B=^NonFocusAngleThreshold
     Out Range B=^NonFocusValue
     In=AbsoluteValue.Abs:Float(Input=^Elevation).Value
    ).Out Value
   )
   n3 = VariableDeferredAccessor.Float(Variable=InitVariable.Float().Variable)
   n4 = VariableDeferredAccessor.Float(Variable=@n3.Variable)
   TriggerRouteTriggerRouteFloat2_0 = "TriggerRoute.Trigger Route (Float, 2):Float"(
    Value 0=UE.Multiply:Float(AdditionalOperands=^FocusAttackInterpSpeed, PrimaryOperand=@n1.Float).Out
    Value 1=UE.Multiply:Float(AdditionalOperands=^FocusReleaseInterpSpeed, PrimaryOperand=@n1.Float).Out
    Set 0=UE.TriggerOnThreshold:Float(In=@n2.Out, Threshold=@n4.Value).Out
    Set 1=UE.TriggerOnThreshold:Float(In=@n2.Out, Threshold=@n3.Value) {also {defaults {Type=1}}}.Out
   )
   InterpTo_0 = UE.InterpTo:Audio(Interp Time=UE.ConversionFloatToTime(In=@TriggerRouteTriggerRouteFloat2_0.Value).Out, Target=@n2.Out)
   VariableMutator.Float(Value=@n2.Out, Variable=@n4.Variable)
  ]
  links [{dst="@FocusMultiplier", src=InterpTo_0.Value}]
 }]
