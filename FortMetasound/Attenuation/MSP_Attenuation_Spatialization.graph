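# MetaSound patch that pans mono and stereo sources to a stereo output by azimuth, with a distance-based non-spatialized radius.
# Stereo channels are offset either side of Azimuth by StereoSpread, while mono feeds two gain-scaled panners without spread.
# Trigger compares wrap each angle within ±180 degrees, and a five-point crossfade converts it to pan.
# Distance crossfades from an unpanned mix to the panned mix between the radius start and end.
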MetaSoundPatch /FortMetasound/Attenuation/MSP_Attenuation_Spatialization
inputs [
 Azimuth {type=Float}
 Distance {type=Float}
 Mono.In {type=Audio}
 NonSpatializedRadiusEnd {type=Float}
 NonSpatializedRadiusStart {type=Float}
 Stereo.InL {type=Audio}
 Stereo.InR {type=Audio}
 StereoSpread {type=Float}
]
outputs [
 Left {type=Audio}
 Right {type=Audio}
]
pages [Default {
  input_defaults {
   Azimuth=0
   Distance=0
   Mono.In=null
   NonSpatializedRadiusEnd=0
   NonSpatializedRadiusStart=0
   Stereo.InL=null
   Stereo.InR=null
   StereoSpread=20
  }
  nodes [
   n1 = UE.Subtract:Float(PrimaryOperand=^Azimuth, AdditionalOperands=^StereoSpread)
   TriggerOnValueChange_1 = UE.Trigger On Value Change:Float(Value=@n1.Out)
   n2 = TriggerCompare.Clamp:Float(A=@n1.Out, Compare=@TriggerOnValueChange_1.Trigger) {also {defaults {B=180, Type=3}}}
   TriggerRouteTriggerRouteBool2_3 = "TriggerRoute.Trigger Route (Bool, 2):Bool"(Set 1=@n2.False, Set 0=@n2.True) {also {defaults {Value 0="<MetaSoundBuilderResult.SUCCEEDED: 0>"}}}
   n3 = UE.Add:Float(
    AdditionalOperands=UE.Multiply:Float(
     PrimaryOperand=Convert.Float:Bool(Bool=@TriggerRouteTriggerRouteBool2_3.Value).Float
     AdditionalOperands=UE.Multiply:Float() {also {defaults {AdditionalOperands=-2, PrimaryOperand=180}}}.Out
    ).Out
    PrimaryOperand=@n1.Out
   )
   TriggerOnValueChange_0 = UE.Trigger On Value Change:Float(Value=@n1.Out)
   n4 = TriggerCompare.Clamp:Float(A=@n1.Out, Compare=@TriggerOnValueChange_0.Trigger) {also {defaults {B=-180, Type=2}}}
   TriggerRouteTriggerRouteBool2_0 = "TriggerRoute.Trigger Route (Bool, 2):Bool"(Set 1=@n4.False, Set 0=@n4.True) {also {defaults {Value 0="<MetaSoundBuilderResult.SUCCEEDED: 0>"}}}
   n5 = UE.Multiply:Float(
    PrimaryOperand=Convert.Float:Bool(Bool=@TriggerRouteTriggerRouteBool2_0.Value).Float
    AdditionalOperands=UE.Multiply:Float() {also {defaults {AdditionalOperands=2, PrimaryOperand=180}}}.Out
   )
   CrossfadeTriggerRouteFloat5_2 = "Crossfade.Trigger Route (Float, 5):Float"(Crossfade Value=MapRange.MapRange:Float(In=UE.Add:Float(PrimaryOperand=@n3.Out, AdditionalOperands=@n5.Out).Out) {also {defaults {In Range A=-180, In Range B=180, Out Range B=4}}}.Out Value) {also {defaults {In 0=0, In 1=-1, In 2=0, In 3=1, In 4=0}}}
   StereoPanner_0 = UE.Stereo Panner(In=^Stereo.InL, Pan Amount=@CrossfadeTriggerRouteFloat5_2.Out)
   n6 = UE.Add:Float(PrimaryOperand=^Azimuth, AdditionalOperands=^StereoSpread)
   TriggerOnValueChange_4 = UE.Trigger On Value Change:Float(Value=@n6.Out)
   n7 = TriggerCompare.Clamp:Float(A=@n6.Out, Compare=@TriggerOnValueChange_4.Trigger) {also {defaults {B=180, Type=3}}}
   TriggerRouteTriggerRouteBool2_7 = "TriggerRoute.Trigger Route (Bool, 2):Bool"(Set 1=@n7.False, Set 0=@n7.True) {also {defaults {Value 0="<MetaSoundBuilderResult.SUCCEEDED: 0>"}}}
   n8 = UE.Add:Float(
    PrimaryOperand=@n6.Out
    AdditionalOperands=UE.Multiply:Float(
     PrimaryOperand=Convert.Float:Bool(Bool=@TriggerRouteTriggerRouteBool2_7.Value).Float
     AdditionalOperands=UE.Multiply:Float() {also {defaults {AdditionalOperands=-2, PrimaryOperand=180}}}.Out
    ).Out
   )
   TriggerOnValueChange_5 = UE.Trigger On Value Change:Float(Value=@n6.Out)
   n9 = TriggerCompare.Clamp:Float(A=@n6.Out, Compare=@TriggerOnValueChange_5.Trigger) {also {defaults {B=-180, Type=2}}}
   TriggerRouteTriggerRouteBool2_4 = "TriggerRoute.Trigger Route (Bool, 2):Bool"(Set 1=@n9.False, Set 0=@n9.True) {also {defaults {Value 0="<MetaSoundBuilderResult.SUCCEEDED: 0>"}}}
   n10 = UE.Multiply:Float(
    PrimaryOperand=Convert.Float:Bool(Bool=@TriggerRouteTriggerRouteBool2_4.Value).Float
    AdditionalOperands=UE.Multiply:Float() {also {defaults {AdditionalOperands=2, PrimaryOperand=180}}}.Out
   )
   CrossfadeTriggerRouteFloat5_3 = "Crossfade.Trigger Route (Float, 5):Float"(Crossfade Value=MapRange.MapRange:Float(In=UE.Add:Float(PrimaryOperand=@n8.Out, AdditionalOperands=@n10.Out).Out) {also {defaults {In Range A=-180, In Range B=180, Out Range B=4}}}.Out Value) {also {defaults {In 0=0, In 1=-1, In 2=0, In 3=1, In 4=0}}}
   StereoPanner_2 = UE.Stereo Panner(In=^Stereo.InR, Pan Amount=@CrossfadeTriggerRouteFloat5_3.Out)
   AudioMixerAudioMixerStereo2_1 = "AudioMixer.Audio Mixer (Stereo, 2)"(In 0 L=@StereoPanner_0.Out Left, In 0 R=@StereoPanner_0.Out Right, In 1 L=@StereoPanner_2.Out Left, In 1 R=@StereoPanner_2.Out Right)
   n11 = UE.Subtract:Float(PrimaryOperand=^Azimuth)
   TriggerOnValueChange_7 = UE.Trigger On Value Change:Float(Value=@n11.Out)
   n12 = TriggerCompare.Clamp:Float(A=@n11.Out, Compare=@TriggerOnValueChange_7.Trigger) {also {defaults {B=180, Type=3}}}
   TriggerRouteTriggerRouteBool2_1 = "TriggerRoute.Trigger Route (Bool, 2):Bool"(Set 1=@n12.False, Set 0=@n12.True) {also {defaults {Value 0="<MetaSoundBuilderResult.SUCCEEDED: 0>"}}}
   n13 = UE.Add:Float(
    AdditionalOperands=UE.Multiply:Float(
     PrimaryOperand=Convert.Float:Bool(Bool=@TriggerRouteTriggerRouteBool2_1.Value).Float
     AdditionalOperands=UE.Multiply:Float() {also {defaults {AdditionalOperands=-2, PrimaryOperand=180}}}.Out
    ).Out
    PrimaryOperand=@n11.Out
   )
   TriggerOnValueChange_2 = UE.Trigger On Value Change:Float(Value=@n11.Out)
   n14 = TriggerCompare.Clamp:Float(A=@n11.Out, Compare=@TriggerOnValueChange_2.Trigger) {also {defaults {B=-180, Type=2}}}
   TriggerRouteTriggerRouteBool2_6 = "TriggerRoute.Trigger Route (Bool, 2):Bool"(Set 1=@n14.False, Set 0=@n14.True) {also {defaults {Value 0="<MetaSoundBuilderResult.SUCCEEDED: 0>"}}}
   n15 = UE.Multiply:Float(
    PrimaryOperand=Convert.Float:Bool(Bool=@TriggerRouteTriggerRouteBool2_6.Value).Float
    AdditionalOperands=UE.Multiply:Float() {also {defaults {AdditionalOperands=2, PrimaryOperand=180}}}.Out
   )
   CrossfadeTriggerRouteFloat5_1 = "Crossfade.Trigger Route (Float, 5):Float"(Crossfade Value=MapRange.MapRange:Float(In=UE.Add:Float(PrimaryOperand=@n13.Out, AdditionalOperands=@n15.Out).Out) {also {defaults {In Range A=-180, In Range B=180, Out Range B=4}}}.Out Value) {also {defaults {In 0=0, In 1=-1, In 2=0, In 3=1, In 4=0}}}
   n16 = VariableAccessor.Float(Variable=InitVariable.Float().Variable)
   StereoPanner_1 = UE.Stereo Panner(Pan Amount=@CrossfadeTriggerRouteFloat5_1.Out, In=UE.Multiply:Audio by Float(PrimaryOperand=^Mono.In, AdditionalOperands=@n16.Value).Out)
   n17 = UE.Add:Float(PrimaryOperand=^Azimuth)
   TriggerOnValueChange_6 = UE.Trigger On Value Change:Float(Value=@n17.Out)
   n18 = TriggerCompare.Clamp:Float(A=@n17.Out, Compare=@TriggerOnValueChange_6.Trigger) {also {defaults {B=180, Type=3}}}
   TriggerRouteTriggerRouteBool2_5 = "TriggerRoute.Trigger Route (Bool, 2):Bool"(Set 1=@n18.False, Set 0=@n18.True) {also {defaults {Value 0="<MetaSoundBuilderResult.SUCCEEDED: 0>"}}}
   n19 = UE.Add:Float(
    PrimaryOperand=@n17.Out
    AdditionalOperands=UE.Multiply:Float(
     PrimaryOperand=Convert.Float:Bool(Bool=@TriggerRouteTriggerRouteBool2_5.Value).Float
     AdditionalOperands=UE.Multiply:Float() {also {defaults {AdditionalOperands=-2, PrimaryOperand=180}}}.Out
    ).Out
   )
   TriggerOnValueChange_3 = UE.Trigger On Value Change:Float(Value=@n17.Out)
   n20 = TriggerCompare.Clamp:Float(A=@n17.Out, Compare=@TriggerOnValueChange_3.Trigger) {also {defaults {B=-180, Type=2}}}
   TriggerRouteTriggerRouteBool2_2 = "TriggerRoute.Trigger Route (Bool, 2):Bool"(Set 1=@n20.False, Set 0=@n20.True) {also {defaults {Value 0="<MetaSoundBuilderResult.SUCCEEDED: 0>"}}}
   n21 = UE.Multiply:Float(
    PrimaryOperand=Convert.Float:Bool(Bool=@TriggerRouteTriggerRouteBool2_2.Value).Float
    AdditionalOperands=UE.Multiply:Float() {also {defaults {AdditionalOperands=2, PrimaryOperand=180}}}.Out
   )
   CrossfadeTriggerRouteFloat5_0 = "Crossfade.Trigger Route (Float, 5):Float"(Crossfade Value=MapRange.MapRange:Float(In=UE.Add:Float(PrimaryOperand=@n19.Out, AdditionalOperands=@n21.Out).Out) {also {defaults {In Range A=-180, In Range B=180, Out Range B=4}}}.Out Value) {also {defaults {In 0=0, In 1=-1, In 2=0, In 3=1, In 4=0}}}
   StereoPanner_3 = UE.Stereo Panner(Pan Amount=@CrossfadeTriggerRouteFloat5_0.Out, In=UE.Multiply:Audio by Float(PrimaryOperand=^Mono.In, AdditionalOperands=@n16.Value).Out)
   AudioMixerAudioMixerStereo2_2 = "AudioMixer.Audio Mixer (Stereo, 2)"(In 0 L=@StereoPanner_1.Out Left, In 0 R=@StereoPanner_1.Out Right, In 1 L=@StereoPanner_3.Out Left, In 1 R=@StereoPanner_3.Out Right)
   AudioMixerAudioMixerStereo2_0 = "AudioMixer.Audio Mixer (Stereo, 2)"(
    In 1 L=@AudioMixerAudioMixerStereo2_1.Out L
    In 1 R=@AudioMixerAudioMixerStereo2_1.Out R
    In 0 L=@AudioMixerAudioMixerStereo2_2.Out L
    In 0 R=@AudioMixerAudioMixerStereo2_2.Out R
   )
   AudioMixerAudioMixerStereo2_3 = "AudioMixer.Audio Mixer (Stereo, 2)"(In 0 L=^Mono.In, In 0 R=^Mono.In, In 1 L=^Stereo.InL, In 1 R=^Stereo.InR) {also {defaults {Gain 0=0.55}}}
   n22 = UE.Add:Float(PrimaryOperand=^NonSpatializedRadiusEnd) {also {defaults {AdditionalOperands=0.1}}}
   n23 = VariableAccessor.Float(Variable=VariableMutator.Float(
     Variable=InitVariable.Float().Variable
     Value=MapRange.MapRange:Float(In=^Distance, In Range A=^NonSpatializedRadiusStart, In Range B=@n22.Out).Out Value
    ).Variable)
   CrossfadeTriggerRouteAudio2_0 = "Crossfade.Trigger Route (Audio, 2):Audio"(In 1=@AudioMixerAudioMixerStereo2_0.Out L, In 0=@AudioMixerAudioMixerStereo2_3.Out L, Crossfade Value=@n23.Value)
   CrossfadeTriggerRouteAudio2_1 = "Crossfade.Trigger Route (Audio, 2):Audio"(In 1=@AudioMixerAudioMixerStereo2_0.Out R, In 0=@AudioMixerAudioMixerStereo2_3.Out R, Crossfade Value=@n23.Value)
  ]
  links [{dst="@Left", src=CrossfadeTriggerRouteAudio2_0.Out}, {dst="@Right", src=CrossfadeTriggerRouteAudio2_1.Out}]
 }]
