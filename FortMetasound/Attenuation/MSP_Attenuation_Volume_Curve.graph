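# MetaSound patch that computes a distance attenuation volume multiplier from a curve in a WaveTable bank.
# It maps Distance from the span between InnerRadius and InnerRadius plus FalloffDistance into a curve input, then evaluates the bank curve chosen by CurveIndexToEvaluate.
# The result is multiplied by bEnable converted to float, so disabling it outputs zero rather than bypassing attenuation.
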MetaSoundPatch /FortMetasound/Attenuation/MSP_Attenuation_Volume_Curve
inputs [
 AttenuationCurves {type=WaveTableBankAsset}
 bEnable {type=Bool}
 CurveIndexToEvaluate {type=Float}
 Distance {type=Float}
 FalloffDistance {type=Float}
 InnerRadius {type=Float}
]
outputs [VolumeMultiplier {type=Float}]
pages [Default {
  input_defaults {
   AttenuationCurves=null
   CurveIndexToEvaluate=0
   Distance=0
   FalloffDistance=1000
   InnerRadius=0
   bEnable="<MetaSoundBuilderResult.SUCCEEDED: 0>"
  }
  nodes [
   n1 = UE.Add:Float(AdditionalOperands=^FalloffDistance, PrimaryOperand=^InnerRadius)
   Multiply_0 = UE.Multiply:Float(
    PrimaryOperand=Convert.Float:Bool(Bool=^bEnable).Float
    AdditionalOperands=UE.WaveTableBankEvaluate(
     WaveTableBank=^AttenuationCurves
     Index=^CurveIndexToEvaluate
     Input=MapRange.MapRange:Float(In=^Distance, In Range A=^InnerRadius, In Range B=@n1.Out).Out Value
    ).Output
   )
  ]
  links [{dst="@VolumeMultiplier", src=Multiply_0.Out}]
 }]
