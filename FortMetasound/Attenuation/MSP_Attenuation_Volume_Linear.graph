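# MetaSound patch that computes a linear distance-based volume multiplier for audio attenuation.
# Distance is mapped from InnerRadius to InnerRadius plus FalloffDistance, giving a multiplier that falls from 1 to 0.
# The result is multiplied by bEnable converted to a float, so disabling the patch outputs zero volume rather than bypassing attenuation.
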MetaSoundPatch /FortMetasound/Attenuation/MSP_Attenuation_Volume_Linear
inputs [
 bEnable {type=Bool}
 Distance {type=Float}
 FalloffDistance {type=Float}
 InnerRadius {type=Float}
]
outputs [VolumeMultiplier {type=Float}]
pages [Default {
  input_defaults {Distance=0, FalloffDistance=1000, InnerRadius=0, bEnable="<MetaSoundBuilderResult.SUCCEEDED: 0>"}
  nodes [Multiply_0 = UE.Multiply:Float(
    PrimaryOperand=Convert.Float:Bool(Bool=^bEnable).Float
    AdditionalOperands=MapRange.MapRange:Float(In=^Distance, In Range A=^InnerRadius, In Range B=UE.Add:Float(AdditionalOperands=^FalloffDistance, PrimaryOperand=^InnerRadius).Out) {also {defaults {Out Range A=1, Out Range B=0}}}.Out Value
   )]
  links [{dst="@VolumeMultiplier", src=Multiply_0.Out}]
 }]
